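# Control Rig for the male Space Warrior cosmetic body that procedurally adjusts armor bones to follow body motion.
# Clavicle and head rotations relative to spine_05 are converted to Euler angles and remapped with clamping into driver values.
# These drive additive local offsets on the left shoulder pad, then the collar.
# A right shoulder pad branch exists but is disconnected from execution and reads unset bones.
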ControlRigBlueprint /BRCosmetics/Characters/Player/Male/Medium/Bodies/M_MED_Space_Warrior/Meshes/Space_Warrior_CR
graphs [
 RigVMFunctionLibrary {links [], nodes []}
 RigVMModel {nodes [
   RigUnit_BeginExecution = &RigUnit_BeginExecution()
   n1 = &RigUnit_GetRelativeBoneTransform(Bone=clavicle_l, Space=spine_05)
   n2 = &RigVMFunction_MathFloatRemap(
    SourceMaximum=-52
    SourceMinimum=-80
    bClamp=true
    Value=/Script/RigVM.RigVMFunction_MathFloatNegate(Value=&RigVMFunction_MathQuaternionToEuler(RotationOrder=ZXY, Value=@n1.Transform.Rotation).Result.Y) {is=struct RigVMUnitNode, template="Negate::Execute(in Value,out Result)"}.Result
   )
   /Script/ControlRig.RigUnit_ModifyBoneTransforms(
    BoneToModify=((Bone="dyn_shoulderpad_l",Transform=(Rotation=(X=0,Y=0,Z=0,W=1),Translation=(X=0,Y=0,Z=0),Scale3D=(X=1,Y=1,Z=1))))
    Mode=AdditiveLocal
    Weight=1
    WeightMaximum=1
    WeightMinimum=0
    BoneToModify.0.Transform.Translation.X=&RigVMFunction_MathFloatMul(B=10, A=@n2.Result) {template="Multiply::Execute(in A,in B,out Result)"}.Result
    BoneToModify.0.Transform.Translation.Y=&RigVMFunction_MathFloatMul(B=-3, A=@n2.Result) {template="Multiply::Execute(in A,in B,out Result)"}.Result
    BoneToModify.0.Transform.Translation.Z=&RigVMFunction_MathFloatMul(B=-2, A=@n2.Result) {template="Multiply::Execute(in A,in B,out Result)"}.Result
    BoneToModify.0.Transform.Rotation=&RigVMFunction_MathQuaternionFromEuler(
     Euler=(X=0,Y=0,Z=0)
     Euler.X=&RigVMFunction_MathFloatMul(B=15, A=@n2.Result) {template="Multiply::Execute(in A,in B,out Result)"}.Result
    ).Result
    ExecutePin=@RigUnit_BeginExecution.ExecutePin
   ) {is=struct RigVMUnitNode}
   n3 = &RigUnit_GetRelativeBoneTransform(Bone=head, Space=spine_05)
   n4 = &RigVMFunction_MathFloatRemap(
    SourceMaximum=40
    SourceMinimum=0
    TargetMinimum=0
    bClamp=true
    Value=&RigVMFunction_MathFloatMul(B=-1, A=&RigVMFunction_MathQuaternionToEuler(RotationOrder=ZXY, Value=@n3.Transform.Rotation).Result.Z) {template="Multiply::Execute(in A,in B,out Result)"}.Result
   )
   > &RigUnit_ModifyBoneTransforms(
    BoneToModify=((Bone="dyn_collar",Transform=(Rotation=(X=0,Y=0,Z=0,W=1),Translation=(X=0,Y=0,Z=0),Scale3D=(X=1,Y=1,Z=1))))
    Mode=AdditiveLocal
    Weight=1
    WeightMaximum=1
    WeightMinimum=0
    BoneToModify.0.Transform.Translation.Y=&RigVMFunction_MathFloatMul(B=-0.5, A=@n4.Result) {template="Multiply::Execute(in A,in B,out Result)"}.Result
    BoneToModify.0.Transform.Rotation=&RigVMFunction_MathQuaternionFromEuler(
     Euler=(X=0,Y=0,Z=0)
     Euler.Z=&RigVMFunction_MathFloatMul(B=-35, A=@n4.Result) {template="Multiply::Execute(in A,in B,out Result)"}.Result
    ).Result
   )
   n5 = &RigVMFunction_MathFloatNegate(Value=&RigVMFunction_MathQuaternionToEuler(RotationOrder=ZXY, Value=&RigUnit_GetRelativeBoneTransform(Bone=None, Space=None).Transform.Rotation).Result.X) {template="Negate::Execute(in Value,out Result)"}
   n6 = &RigVMFunction_MathFloatRemap(
    SourceMaximum=0
    SourceMinimum=-17
    bClamp=true
    Value=&RigVMFunction_MathQuaternionToEuler(RotationOrder=ZXY, Value=&RigUnit_GetRelativeBoneTransform(Bone=None, Space=None).Transform.Rotation).Result.X
   )
   &RigUnit_ModifyBoneTransforms(
    BoneToModify=((Bone="dyn_shoulderpad_r",Transform=(Rotation=(X=0,Y=0,Z=0,W=1),Translation=(X=0,Y=0,Z=0),Scale3D=(X=1,Y=1,Z=1))))
    ExecutePin=()
    Mode=AdditiveLocal
    Weight=1
    WeightMaximum=1
    WeightMinimum=0
    BoneToModify.0.Transform.Translation.X=&RigVMFunction_MathFloatMul(B=-5, A=&RigVMFunction_MathFloatRemap(SourceMaximum=5, SourceMinimum=0, bClamp=true, Value=@n5.Result).Result) {template="Multiply::Execute(in A,in B,out Result)"}.Result
    BoneToModify.0.Transform.Translation.Y=&RigVMFunction_MathFloatMul(B=8, A=@n6.Result) {template="Multiply::Execute(in A,in B,out Result)"}.Result
    BoneToModify.0.Transform.Rotation=&RigVMFunction_MathQuaternionFromEuler(
     Euler=(X=0,Y=0,Z=0)
     Euler.X=&RigVMFunction_MathFloatMul(B=45, A=@n6.Result) {template="Multiply::Execute(in A,in B,out Result)"}.Result
    ).Result
   )
  ]}
]
hierarchy [
 root {loc=<0 0 0>, rot=<0 0 0 1>}
 pelvis {parent=root, loc=<0 -0.7071 93.1191>, rot=<0.0127 0.707 -0.0127 -0.707>}
 spine_01 {parent=pelvis, loc=<6.7494 -0.6388 0>, rot=<0 0 0.034 -0.9994>}
 spine_02 {parent=spine_01, loc=<7.2518 -0.035 0>, rot=<0 0 0.0238 -0.9997>}
 spine_03 {parent=spine_02, loc=<6.0104 0 0>, rot=<0 0 0.0149 -0.9999>}
 spine_04 {parent=spine_03, loc=<8.1024 0 0>, rot=<0 0 -0.0132 -0.9999>}
 spine_05 {parent=spine_04, loc=<10.0146 0 0>, rot=<0 0 -0.0015 -1>}
 clavicle_l {parent=spine_05, loc=<4.3575 -0.1466 -2.3223>, rot=<0.0465 -0.7289 -0.009 -0.6829>}
 upperarm_l {parent=clavicle_l, loc=<15.5111 0 0>, rot=<0.0226 -0.3234 -0.0088 -0.9459>}
 lowerarm_l {parent=upperarm_l, loc=<26.6514 0 0>, rot=<0.0391 -0.0099 -0.2453 -0.9686>}
 hand_l {parent=lowerarm_l, loc=<22.3491 0 0>, rot=<-0.7081 -0.0507 0.0975 -0.6975>}
 index_metacarpal_l {parent=hand_l, loc=<1.6535 0.0183 -2.96>, rot=<0.0158 -0.0124 -0.0173 -0.9996>}
 index_01_l {parent=index_metacarpal_l, loc=<8.3008 0.0357 -0.0524>, rot=<-0.0044 0.0485 0.0924 -0.9945>}
 index_02_l {parent=index_01_l, loc=<3.3587 -0.3004 0.0452>, rot=<-0.002 0.0027 0.3225 -0.9466>}
 index_03_l {parent=index_02_l, loc=<2.5578 -0.123 -0.0252>, rot=<-0.002 0.0067 0.2558 -0.9667>}
 middle_metacarpal_l {parent=hand_l, loc=<2.1856 -0.1725 -0.5984>, rot=<0.006 0.0168 -0.0329 -0.9993>}
 middle_01_l {parent=middle_metacarpal_l, loc=<7.3275 -0.2447 0.0129>, rot=<-0.0208 0.0042 0.0858 -0.9961>}
 middle_02_l {parent=middle_01_l, loc=<4.0123 0.1401 0.1175>, rot=<-0.0218 0.0258 0.3512 -0.9357>}
 middle_03_l {parent=middle_02_l, loc=<2.6077 0.0534 0.1373>, rot=<0.0082 0.0188 0.1708 -0.9851>}
 pinky_metacarpal_l {parent=hand_l, loc=<1.7616 -0.9753 3.3814>, rot=<-0.0397 0.0814 -0.069 -0.9935>}
 pinky_01_l {parent=pinky_metacarpal_l, loc=<7.1709 -0.4831 -0.1094>, rot=<-0.0654 -0.0035 0.2242 -0.9723>}
 pinky_02_l {parent=pinky_01_l, loc=<2.8274 0.0893 -0.1043>, rot=<0.0037 0.0011 0.2854 -0.9584>}
 pinky_03_l {parent=pinky_02_l, loc=<2.2046 0.0307 -0.0078>, rot=<-0.0002 0.0098 0.1445 -0.9895>}
 ring_metacarpal_l {parent=hand_l, loc=<2.2074 -0.3625 1.3638>, rot=<-0.0284 0.0722 -0.0458 -0.9959>}
 ring_01_l {parent=ring_metacarpal_l, loc=<7.0295 -0.6197 -0.1446>, rot=<-0.0489 -0.0252 0.1892 -0.9804>}
 ring_02_l {parent=ring_01_l, loc=<3.6174 0.4646 -0.1027>, rot=<0 0 0.2806 -0.9598>}
 ring_03_l {parent=ring_02_l, loc=<2.4152 0.1141 0.0199>, rot=<0 0 0.1189 -0.9929>}
 thumb_01_l {parent=hand_l, loc=<1.3118 -0.3191 -2.459>, rot=<0.6546 -0.4042 0.0664 -0.6354>}
 thumb_02_l {parent=thumb_01_l, loc=<5.9336 -0.0448 0.0401>, rot=<-0.0037 0.0213 0.298 -0.9543>}
 thumb_03_l {parent=thumb_02_l, loc=<2.2223 0.3682 -0.0907>, rot=<0.0118 -0.013 0.0066 -0.9998>}
 lowerarm_twist_01_l {parent=lowerarm_l, loc=<17.8549 0 0>, rot=<0 0 0 1>}
 lowerarm_twist_02_l {parent=lowerarm_l, loc=<9.6199 0 0>, rot=<0 0 0 1>}
 upperarm_twist_01_l {parent=upperarm_l, loc=<0.0041 0 0>, rot=<0 0 0 -1>}
 upperarm_twist_02_l {parent=upperarm_l, loc=<13.7941 0 0>, rot=<0 0 0 -1>}
 dyn_shoulderpad_l {parent=clavicle_l, loc=<12.8348 -1.5131 11.1023>, rot=<0.0172 -0.1872 -0.0264 -0.9818>}
 clavicle_r {parent=spine_05, loc=<4.3575 -0.1466 2.3223>, rot=<0.7289 0.0465 -0.6829 0.009>}
 upperarm_r {parent=clavicle_r, loc=<-15.5111 0 0>, rot=<0.0226 -0.3234 -0.0088 -0.9459>}
 lowerarm_r {parent=upperarm_r, loc=<-26.6514 0 0>, rot=<0.0391 -0.0099 -0.2453 -0.9686>}
 hand_r {parent=lowerarm_r, loc=<-22.3492 0 0>, rot=<-0.7081 -0.0507 0.0975 -0.6975>}
 index_metacarpal_r {parent=hand_r, loc=<-1.6534 -0.0181 2.9601>, rot=<0.0158 -0.0124 -0.0173 -0.9996>}
 index_01_r {parent=index_metacarpal_r, loc=<-8.3008 -0.0357 0.0524>, rot=<-0.0044 0.0485 0.0924 -0.9945>}
 index_02_r {parent=index_01_r, loc=<-3.3589 0.3003 -0.0453>, rot=<-0.002 0.0027 0.3225 -0.9466>}
 index_03_r {parent=index_02_r, loc=<-2.5578 0.1231 0.0251>, rot=<-0.002 0.0067 0.2558 -0.9667>}
 middle_metacarpal_r {parent=hand_r, loc=<-2.1856 0.1726 0.5984>, rot=<0.006 0.0168 -0.0329 -0.9993>}
 middle_01_r {parent=middle_metacarpal_r, loc=<-7.3273 0.2447 -0.0129>, rot=<-0.0208 0.0042 0.0858 -0.9961>}
 middle_02_r {parent=middle_01_r, loc=<-4.0142 -0.1411 -0.07>, rot=<-0.0217 0.0259 0.3512 -0.9357>}
 middle_03_r {parent=middle_02_r, loc=<-2.6074 -0.0534 -0.1312>, rot=<0.0083 0.0188 0.1708 -0.9851>}
 pinky_metacarpal_r {parent=hand_r, loc=<-1.7617 0.9753 -3.3814>, rot=<-0.0397 0.0814 -0.069 -0.9935>}
 pinky_01_r {parent=pinky_metacarpal_r, loc=<-7.1707 0.4832 0.1094>, rot=<-0.0654 -0.0035 0.2242 -0.9723>}
 pinky_02_r {parent=pinky_01_r, loc=<-2.8274 -0.0892 0.1043>, rot=<0.0037 0.0011 0.2854 -0.9584>}
 pinky_03_r {parent=pinky_02_r, loc=<-2.2045 -0.031 0.0078>, rot=<-0.0002 0.0098 0.1445 -0.9895>}
 ring_metacarpal_r {parent=hand_r, loc=<-2.2073 0.3627 -1.3638>, rot=<-0.0284 0.0722 -0.0458 -0.9959>}
 ring_01_r {parent=ring_metacarpal_r, loc=<-7.0297 0.6197 0.1446>, rot=<-0.0489 -0.0252 0.1892 -0.9804>}
 ring_02_r {parent=ring_01_r, loc=<-3.6172 -0.4648 0.1026>, rot=<0 0 0.2806 -0.9598>}
 ring_03_r {parent=ring_02_r, loc=<-2.4152 -0.1141 -0.0199>, rot=<0 0 0.1189 -0.9929>}
 thumb_01_r {parent=hand_r, loc=<-1.3119 0.3192 2.459>, rot=<0.6546 -0.4042 0.0664 -0.6354>}
 thumb_02_r {parent=thumb_01_r, loc=<-5.9335 0.0447 -0.0401>, rot=<-0.0037 0.0213 0.298 -0.9543>}
 thumb_03_r {parent=thumb_02_r, loc=<-2.2223 -0.3682 0.0907>, rot=<0.0118 -0.013 0.0066 -0.9998>}
 lowerarm_twist_01_r {parent=lowerarm_r, loc=<-17.855 0 0>, rot=<0 0 0 -1>}
 lowerarm_twist_02_r {parent=lowerarm_r, loc=<-9.6199 0 0>, rot=<0 0 0 -1>}
 upperarm_twist_01_r {parent=upperarm_r, loc=<-0.0041 0 0>, rot=<0 0 0 1>}
 upperarm_twist_02_r {parent=upperarm_r, loc=<-13.7942 0 0>, rot=<0 0 0 1>}
 dyn_shoulderpad_r {parent=clavicle_r, loc=<-12.8057 1.5146 -11.1003>, rot=<0.0172 -0.1872 -0.0264 -0.9818>}
 neck_01 {parent=spine_05, loc=<11.3138 0.1529 0>, rot=<0 0 -0.2003 -0.9797>}
 neck_02 {parent=neck_01, loc=<4.5627 0 0>, rot=<0 0 0.0504 -0.9987>}
 head {parent=neck_02, loc=<5.3914 0 0>, rot=<0 0 0.115 -0.9934>}
 dyn_grenade_01 {parent=spine_05, loc=<-100.9782 -13.2557 -20.5802>, rot=<0.0633 -0.7431 0.0486 -0.6644>}
 dyn_grenade_02 {parent=dyn_grenade_01, loc=<-47.9928 18.8643 102.3699>, rot=<-0.1557 0.8064 0.0087 -0.5704>}
 dyn_collar {parent=spine_05, loc=<18.2329 -12.0805 0>, rot=<0 0 -0.4257 -0.9049>}
 dyn_belt_01_master {parent=spine_01, loc=<0.5851 6.0267 -14.1796>, rot=<-0.0897 -0.6704 -0.3001 -0.6726>}
 dyn_belt_01 {parent=dyn_belt_01_master, loc=<0 0 0>, rot=<0 0 0 1>}
 dyn_belt_02_master {parent=spine_01, loc=<2.322 -1.2134 -14.7519>, rot=<0.0986 -0.6912 -0.0381 -0.7149>}
 dyn_belt_02 {parent=dyn_belt_02_master, loc=<0 0 0>, rot=<0 0 0 1>}
 dyn_belt_03_master {parent=spine_01, loc=<2.9837 -8.0563 -11.5347>, rot=<0.2565 -0.6107 0.1161 -0.7401>}
 dyn_belt_03 {parent=dyn_belt_03_master, loc=<0 0 0>, rot=<0 0 0 1>}
 dyn_belt_04_master {parent=spine_01, loc=<4.0438 -11.8915 -6.6283>, rot=<0.4582 -0.5187 0.2328 -0.6832>}
 dyn_belt_04 {parent=dyn_belt_04_master, loc=<0 0 0>, rot=<0 0 0 -1>}
 dyn_belt_05_master {parent=spine_01, loc=<4.2052 -12.3175 -4.1678>, rot=<0.5767 -0.382 0.4083 -0.5957>}
 dyn_belt_05 {parent=dyn_belt_05_master, loc=<0 0 0>, rot=<0 0 0 1>}
 dyn_belt_06_master {parent=spine_01, loc=<4.2641 -12.9219 -0.6757>, rot=<0.5971 -0.3982 0.3926 -0.5751>}
 dyn_belt_06 {parent=dyn_belt_06_master, loc=<0 0 0>, rot=<0 0 0 1>}
 dyn_belt_07_master {parent=spine_01, loc=<5.4962 -12.9567 2.7367>, rot=<0.6138 -0.4067 0.4468 -0.5082>}
 dyn_belt_07 {parent=dyn_belt_07_master, loc=<0 0 0>, rot=<0 0 0 -1>}
 dyn_belt_08_master {parent=spine_01, loc=<5.117 -11.5416 6.8294>, rot=<0.6465 -0.3086 0.4999 -0.4867>}
 dyn_belt_08 {parent=dyn_belt_08_master, loc=<0 0 0>, rot=<0 0 0 1>}
 dyn_belt_09_master {parent=spine_01, loc=<3.9697 -7.8095 11.4328>, rot=<0.7489 -0.1202 0.601 -0.2521>}
 dyn_belt_09 {parent=dyn_belt_09_master, loc=<0 0 0>, rot=<0 0 0 -1>}
 dyn_belt_10_master {parent=spine_01, loc=<3.4467 -0.9377 14.2969>, rot=<0.7251 0.0312 0.6821 -0.0896>}
 dyn_belt_10 {parent=dyn_belt_10_master, loc=<0 0 0>, rot=<0 0 0 -1>}
 dyn_belt_11_master {parent=spine_01, loc=<1.3589 6.8104 14.0521>, rot=<0.7018 0.243 0.6693 0.0205>}
 dyn_belt_11 {parent=dyn_belt_11_master, loc=<0 0 0>, rot=<0 0 0 -1>}
 thigh_l {parent=pelvis, loc=<-2.5693 -0.0109 -7.7501>, rot=<0.0012 0.0572 0.0155 -0.9982>}
 calf_l {parent=thigh_l, loc=<-39.0484 0 0>, rot=<0.0006 -0.0161 -0.0391 -0.9991>}
 calf_twist_01_l {parent=calf_l, loc=<-35.8131 -0.0967 0.0397>, rot=<0.0237 0.0095 -0.0035 -0.9997>}
 calf_twist_02_l {parent=calf_l, loc=<-18.8064 -0.0967 0.0397>, rot=<0.0237 0.0095 -0.0035 -0.9997>}
 foot_l {parent=calf_l, loc=<-42.561 0.1077 0.4292>, rot=<0.0418 -0.0455 0.0343 -0.9975>}
 ball_l {parent=foot_l, loc=<-6.1599 13.8264 0.6806>, rot=<0.0001 -0.0001 -0.7063 -0.7079>}
 dyn_knife {parent=calf_l, loc=<-23.4859 -5.7462 -6.5298>, rot=<0.0036 -0.7726 0.0987 -0.6272>}
 thigh_twist_01_l {parent=thigh_l, loc=<-21.6303 0 0>, rot=<0.0013 0 0 -1>}
 dyn_pouch_l {parent=thigh_l, loc=<-7.59 0.2187 -9.7508>, rot=<0.0039 -0.7524 -0.0383 -0.6576>}
 thigh_r {parent=pelvis, loc=<-2.5691 -0.0109 7.7501>, rot=<0.0572 -0.0012 0.9982 0.0155>}
 calf_r {parent=thigh_r, loc=<39.0487 0 0>, rot=<0.0006 -0.0161 -0.0391 -0.9991>}
 calf_twist_01_r {parent=calf_r, loc=<35.8132 0.0967 -0.0397>, rot=<0.0237 0.0095 -0.0035 -0.9997>}
 calf_twist_02_r {parent=calf_r, loc=<18.8064 0.0967 -0.0397>, rot=<0.0237 0.0095 -0.0035 -0.9997>}
 foot_r {parent=calf_r, loc=<42.5611 -0.1077 -0.4292>, rot=<0.0418 -0.0455 0.0343 -0.9975>}
 ball_r {parent=foot_r, loc=<6.1598 -13.8264 -0.6806>, rot=<0.0001 -0.0001 -0.7063 -0.7079>}
 thigh_twist_01_r {parent=thigh_r, loc=<21.6303 0 0>, rot=<0.0013 0 0 -1>}
 dyn_pouch_r {parent=thigh_r, loc=<7.5711 0.0007 9.7633>, rot=<-0.0614 0.6572 -0.0242 -0.7508>}
 dyn_front_flap_master {parent=pelvis, loc=<2.558 12.6671 0>, rot=<0.0177 -0.9998 0 0>}
 dyn_front_flap_ {parent=dyn_front_flap_master, loc=<0 0 0>, rot=<0 0 0 -1>}
 dyn_back_flap_master {parent=pelvis, loc=<5.6873 -14.8318 0>, rot=<-0.0324 -0.9995 0 0>}
 dyn_back_flap {parent=dyn_back_flap_master, loc=<0 0 0>, rot=<0 0 0 -1>}
]
