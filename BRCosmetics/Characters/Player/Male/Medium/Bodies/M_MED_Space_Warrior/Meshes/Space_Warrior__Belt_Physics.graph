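# Physics asset for the Space Warrior medium male player body's belt, defining simulated motion for belt pieces, pouches and front and back flaps.
# Box bodies on dyn_ bones hang from spine and thigh capsules: belt pieces constrain to spine_01, pouches to their matching thighs.
# Flaps constrain to small master-bone boxes with only one swing axis free to move.
# Twist is locked on every joint, and dyn_belt_09 also has both swings locked.
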PhysicsAsset /BRCosmetics/Characters/Player/Male/Medium/Bodies/M_MED_Space_Warrior/Meshes/Space_Warrior__Belt_Physics
bodies {
 dyn_back_flap [{shapeType=Box, extentX=20, extentY=3.4973, extentZ=21.718, center=<11.8237 0 0.0003>}]
 dyn_back_flap_master [{shapeType=Box, extentX=1.01, extentY=1.01, extentZ=1.01, rotation=<rot 0 -90 0>}]
 dyn_belt_01 [{shapeType=Box, extentX=4.3716, extentY=4.8723, extentZ=12.3476, center=<0.9347 0.3478 -3.7102>}]
 dyn_belt_02 [{shapeType=Box, extentX=7.9074, extentY=4.3638, extentZ=12.4106, center=<1.7238 0.1285 -4.0065>}]
 dyn_belt_03 [{shapeType=Box, extentX=5.3673, extentY=4.5989, extentZ=12.7177, center=<1.4832 0.2755 -3.8731>}]
 dyn_belt_04 [{shapeType=Box, extentX=3.3114, extentY=3.5092, extentZ=7.3041, center=<1.5531 0.0563 -1.724>}]
 dyn_belt_05 [{shapeType=Box, extentX=3.2811, extentY=3.5414, extentZ=7.3031, center=<1.5778 -0.0653 -1.7237>}]
 dyn_belt_06 [{shapeType=Box, extentX=3.2843, extentY=3.5504, extentZ=7.2851, center=<1.593 0.2138 -1.7171>}]
 dyn_belt_07 [{shapeType=Box, extentX=4.4038, extentY=3.1789, extentZ=13.5299, center=<0.9368 0.0559 -5.1051>}]
 dyn_belt_08 [{shapeType=Box, extentX=4.7096, extentY=3.2565, extentZ=14.4763, center=<0.8872 -0.3941 -5.3453>}]
 dyn_belt_09 [{shapeType=Box, extentX=5.3679, extentY=4.5954, extentZ=12.7177, center=<1.3575 -0.1769 -4.9291>}]
 dyn_belt_10 [{shapeType=Box, extentX=7.9042, extentY=4.3682, extentZ=12.4106, center=<2.0138 0.0107 -5.2308>}]
 dyn_belt_11 [{shapeType=Box, extentX=4.6427, extentY=8.1244, extentZ=12.3476, center=<1.6085 -0.1616 -4.6851>}]
 dyn_front_flap_ [{shapeType=Box, extentX=20, extentY=3.4973, extentZ=21.718, center=<11.8237 0 0.0003>}]
 dyn_front_flap_master [{shapeType=Box, extentX=1.01, extentY=1.01, extentZ=1.01, rotation=<rot 0 -90 0>}]
 dyn_pouch_l [{shapeType=Box, extentX=13.0755, extentY=7.7144, extentZ=21.7274, center=<2.5791 0.4113 -6.4175>, rotation=<rot -0.034 -90 -4.0445>}]
 dyn_pouch_r [{shapeType=Box, extentX=13.0755, extentY=7.7144, extentZ=21.7274, center=<2.5791 -1.0267 -6.3731>, rotation=<rot 3.5688 -90 -4.0445>}]
 spine_01 [{shapeType=Capsule, length=4.25, radius=11.25, center=<-5.3996 -2.9848 0>, rotation=<rot 0 -90 0>}]
 thigh_l [{shapeType=Capsule, length=19.3064, radius=9.0204, center=<-14.2441 1.5182 -0.0001>, rotation=<rot 90 0 0>}]
 thigh_r [{shapeType=Capsule, length=19.3064, radius=9.0204, center=<14.1764 -0.9513 -0.0004>, rotation=<rot 90 0 0>}]
}
constraints [
 {bone1Name=dyn_belt_01, bone2Name=spine_01, swing1LimitDegrees=5, swing2LimitDegrees=20, twistLimitDegrees=45, twistMotion=Locked}
 {bone1Name=dyn_belt_02, bone2Name=spine_01, swing1LimitDegrees=5, swing2LimitDegrees=20, twistLimitDegrees=45, twistMotion=Locked}
 {bone1Name=dyn_belt_03, bone2Name=spine_01, swing1LimitDegrees=5, swing2LimitDegrees=20, twistLimitDegrees=45, twistMotion=Locked}
 {bone1Name=dyn_belt_04, bone2Name=spine_01, swing1LimitDegrees=5, swing2LimitDegrees=20, twistLimitDegrees=45, twistMotion=Locked}
 {bone1Name=dyn_belt_05, bone2Name=spine_01, swing1LimitDegrees=5, swing2LimitDegrees=20, twistLimitDegrees=45, twistMotion=Locked}
 {bone1Name=dyn_belt_06, bone2Name=spine_01, swing1LimitDegrees=5, swing2LimitDegrees=20, twistLimitDegrees=45, twistMotion=Locked}
 {bone1Name=dyn_belt_07, bone2Name=spine_01, swing1LimitDegrees=5, swing2LimitDegrees=20, twistLimitDegrees=45, twistMotion=Locked}
 {bone1Name=dyn_belt_08, bone2Name=spine_01, swing1LimitDegrees=5, swing2LimitDegrees=20, twistLimitDegrees=45, twistMotion=Locked}
 {
  bone1Name=dyn_belt_09
  bone2Name=spine_01
  swing1LimitDegrees=5
  swing1Motion=Locked
  swing2LimitDegrees=20
  swing2Motion=Locked
  twistLimitDegrees=45
  twistMotion=Locked
 }
 {bone1Name=dyn_belt_10, bone2Name=spine_01, swing1LimitDegrees=5, swing2LimitDegrees=20, twistLimitDegrees=45, twistMotion=Locked}
 {bone1Name=dyn_belt_11, bone2Name=spine_01, swing1LimitDegrees=5, swing2LimitDegrees=20, twistLimitDegrees=45, twistMotion=Locked}
 {bone1Name=dyn_pouch_l, bone2Name=thigh_l, swing1LimitDegrees=20, swing2LimitDegrees=20, twistLimitDegrees=45, twistMotion=Locked}
 {bone1Name=dyn_pouch_r, bone2Name=thigh_r, swing1LimitDegrees=20, swing2LimitDegrees=20, twistLimitDegrees=45, twistMotion=Locked}
 {
  bone1Name=dyn_front_flap_
  bone2Name=dyn_front_flap_master
  swing1LimitDegrees=30
  swing2LimitDegrees=45
  swing2Motion=Locked
  twistLimitDegrees=2
  twistMotion=Locked
 }
 {
  bone1Name=dyn_back_flap
  bone2Name=dyn_back_flap_master
  swing1LimitDegrees=30
  swing2LimitDegrees=45
  swing2Motion=Locked
  twistLimitDegrees=2
  twistMotion=Locked
 }
]
