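# Physics asset for the grenade attachment on the Space Warrior medium male player body cosmetic, apparently for secondary dangle motion on its dyn_ bones.
# A tiny capsule on dyn_grenade_01 acts as the pivot, while a larger offset, rotated capsule on dyn_grenade_02 covers the grenade.
# A single constraint joins the two bodies with tight swing and twist limits, restricting the grenade to a small wobble.
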PhysicsAsset /BRCosmetics/Characters/Player/Male/Medium/Bodies/M_MED_Space_Warrior/Meshes/Space_Warrior__Grenade_Physics
bodies {
 dyn_grenade_01 [{shapeType=Capsule, length=0.505, radius=0.505}]
 dyn_grenade_02 [{shapeType=Capsule, length=7.4587, radius=4.0384, center=<-1.8359 3.7163 0.0003>, rotation=<rot 90 0 0>}]
}
constraints [{bone1Name=dyn_grenade_02, bone2Name=dyn_grenade_01, swing1LimitDegrees=10, swing2LimitDegrees=4, twistLimitDegrees=10}]
