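# Post-process-style AnimBlueprint for the Space Warrior body, adding secondary motion to belts, front flap and grenades over a pose copied from the attached parent mesh.
# Standing uses RigidBody physics and a Control Rig.
# Crouching uses a cached pose combining thigh constraints, crouch physics and Control Rig, plus corrective bone offsets driven by thigh rotations measured each update.
# A crouch bool selects the branch before a weighted grenade constraint.
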AnimBlueprint /BRCosmetics/Characters/Player/Male/Medium/Bodies/M_MED_Space_Warrior/Meshes/M_MED_Space_Warrior_Skeleton_AnimBlueprint
target_skeleton=&/M_MED_Space_Warrior_Skeleton
graphs [
 AnimGraph {nodes [
   Control Rig() {
    alphaBoolBlend {blendOption=Linear}
    controlRigAssetReference {blueprintRigClass=&/SpaceWarrior_Crouch_Walk_CR.SpaceWarrior_Crouch_Walk_CR_C}
   }
   Copy Pose From Mesh() {bUseAttachedParent=true}
   "Transform (Modify) Bone - Bone: dyn_belt_11_master"(Alpha=1, Translation="0,0,0", Rotation="0,0,0", Scale="1,1,1") {
    rotationMode=BMM_Additive
    rotationSpace=BCS_BoneSpace
    scaleMode=BMM_Additive
    scaleSpace=BCS_BoneSpace
    translationMode=BMM_Additive
    translationSpace=BCS_BoneSpace
    alphaBoolBlend {blendOption=Linear}
    alphaScaleBiasClamp {inRange {max=0, min=-100}, outRange {max=0, min=1}}
    boneToModify {boneName=dyn_belt_11_master}
   }
   n1 = "Transform (Modify) Bone - Bone: dyn_belt_11_master"(ComponentPose=Local To Component(LocalPose=Use cached pose 'crouch_base'()), Alpha=1, Translation="-1,0,0", Rotation="0,0,0", Scale="1,1,1") {
    rotationMode=BMM_Additive
    rotationSpace=BCS_BoneSpace
    scaleMode=BMM_Additive
    scaleSpace=BCS_BoneSpace
    translationMode=BMM_Additive
    translationSpace=BCS_BoneSpace
    alphaBoolBlend {blendOption=Linear}
    alphaScaleBiasClamp {inRange {max=0, min=-100}, outRange {max=0, min=1}}
    boneToModify {boneName=dyn_belt_11_master}
    translation {x=-1}
   }
   n2 = "Transform (Modify) Bone - Bone: dyn_front_flap_"(ComponentPose=@n1, Alpha=1, Translation="0,3,2", Rotation="0,-25,-10", Scale="1,1,1") {
    rotationMode=BMM_Additive
    rotationSpace=BCS_BoneSpace
    scaleMode=BMM_Additive
    scaleSpace=BCS_BoneSpace
    translationMode=BMM_Additive
    translationSpace=BCS_BoneSpace
    alphaBoolBlend {blendOption=Linear}
    alphaScaleBiasClamp {inRange {max=0, min=-100}, outRange {max=0, min=1}}
    boneToModify {boneName=dyn_front_flap_}
    rotation {roll=-10, yaw=-25}
    translation {y=3, z=2}
   }
   n3 = "Transform (Modify) Bone - Bone: dyn_belt_11_master"(
    ComponentPose=Local To Component(LocalPose=Use cached pose 'crouch_base'())
    Alpha=Get Thigh Rot R() -> (Thigh Rot R:double)
    Translation="0,0,0"
    Rotation="70,-65,0"
    Scale="1,1,1"
   ) {
    rotationMode=BMM_Additive
    rotationSpace=BCS_BoneSpace
    scaleMode=BMM_Additive
    scaleSpace=BCS_BoneSpace
    translationMode=BMM_Additive
    translationSpace=BCS_BoneSpace
    alphaBoolBlend {blendOption=Linear}
    alphaScaleBiasClamp {bMapRange=true, inRange {max=0, min=-103}, outRange {max=0, min=1}}
    boneToModify {boneName=dyn_belt_11_master}
    rotation {pitch=70, yaw=-65}
   }
   n4 = "Transform (Modify) Bone - Bone: dyn_belt_11_master"(ComponentPose=@n3, Alpha=Get Thigh Roll_R() -> (Thigh Roll_R:double), Translation="3,-6,3", Rotation="-30,40,0", Scale="1,1,1") {
    rotationMode=BMM_Additive
    rotationSpace=BCS_BoneSpace
    scaleMode=BMM_Additive
    scaleSpace=BCS_BoneSpace
    translationMode=BMM_Additive
    translationSpace=BCS_BoneSpace
    alphaBoolBlend {blendOption=Linear}
    alphaScaleBiasClamp {bMapRange=true, inRange {max=0, min=-40}, outRange {max=0, min=1}}
    boneToModify {boneName=dyn_belt_11_master}
    rotation {pitch=-30, yaw=40}
    translation=<3 -6 3>
   }
   n5 = "Transform (Modify) Bone - Bone: dyn_front_flap_"(ComponentPose=@n4, Alpha=1, Translation="0,5,0", Rotation="0.000,15,0", Scale="1,1,1") {
    rotationMode=BMM_Additive
    rotationSpace=BCS_BoneSpace
    scaleMode=BMM_Additive
    scaleSpace=BCS_BoneSpace
    translationMode=BMM_Additive
    translationSpace=BCS_BoneSpace
    alphaBoolBlend {blendOption=Linear}
    alphaScaleBiasClamp {inRange {max=0, min=-100}, outRange {max=0, min=1}}
    boneToModify {boneName=dyn_front_flap_}
    rotation {yaw=15}
    translation {y=5}
   }
   n6 = Blend Poses by bool(
    BlendPose_0=Component To Local(ComponentPose=@n2)
    BlendPose_1=Component To Local(ComponentPose=@n5)
    bActiveValue="float > float"(A=Get Thigh Rot R(), B=-35) {is=title K2Node_CallFunction}
   ) {blendType=Linear}
   n7 = RigidBody(ComponentPose=Local To Component(LocalPose=Copy Pose From Mesh() {bUseAttachedParent=true})) {
    overridePhysicsAsset=&/Space_Warrior__Belt_Physics
    simulationSpace=BaseBoneSpace
    alphaBoolBlend {blendOption=Linear}
    baseBoneRef {boneName=pelvis}
    componentLinearAccScale=<1 1 1>
    componentLinearVelScale=<0.3 0.3>
   }
   n8 = RigidBody(ComponentPose=@n7, Alpha=0.62) {
    bOverrideWorldGravity=true
    overridePhysicsAsset=&/Space_Warrior__Grenade_Physics
    simulationSpace=BaseBoneSpace
    alphaBoolBlend {blendOption=Linear}
    baseBoneRef {boneName=root}
    componentLinearAccScale=<1 1 1>
    componentLinearVelScale=<0.3 0.3 0.3>
    overrideWorldGravity {z=-980}
   }
   n9 = Control Rig(Source=Component To Local(ComponentPose=@n8)) {
    alphaBoolBlend {blendOption=Linear}
    controlRigAssetReference {blueprintRigClass=&/Space_Warrior_CR.Space_Warrior_CR_C}
   }
   n10 = Get bIsCrouching()
   n11 = "Constraint - dyn_grenade_01 (PPPPPPP)"!(
    ComponentPose=Local To Component(LocalPose=Blend Poses by bool(BlendPose_0=@n6, BlendPose_1=@n9, bActiveValue=@n10) {blendType=Linear})
    Alpha=1
    ConstraintWeights_0=0.0104
    ConstraintWeights_1:float=0.018
    ConstraintWeights_2:float=0.006
    ConstraintWeights_3:float=0.4515
    ConstraintWeights_4:float=0.4864
    ConstraintWeights_5:float=0.0211
    ConstraintWeights_6:float=0.0065
   ) -> (Pose) {
    alphaBoolBlend {blendOption=Linear}
    boneToModify {boneName=dyn_grenade_01}
    constraintSetup [
     {targetBone {boneName=upperarm_twist_01_r}}
     {targetBone {boneName=spine_03}}
     {targetBone {boneName=spine_04}}
     {targetBone {boneName=spine_05}}
     {targetBone {boneName=clavicle_r}}
     {targetBone {boneName=neck_02}}
     {targetBone {boneName=neck_01}}
    ]
    constraintWeights [1,1,1,1,1,1,1]
   }
   Output Pose(Result=Component To Local(ComponentPose=@n11)) {layerGroup=None}
   n12 = "Constraint - dyn_front_flap_master (RR)"!(
    ComponentPose=Local To Component(LocalPose=Copy Pose From Mesh() {bUseAttachedParent=true})
    Alpha=1
    ConstraintWeights_0=0.5
    ConstraintWeights_1:float=0.5
   ) -> (Pose) {
    alphaBoolBlend {blendOption=Linear}
    boneToModify {boneName=dyn_front_flap_master}
    constraintSetup [{transformType=Rotation, targetBone {boneName=thigh_l}}, {transformType=Rotation, targetBone {boneName=thigh_r}}]
    constraintWeights [1,1]
   }
   n13 = "Transform (Modify) Bone - Bone: dyn_front_flap_master"(ComponentPose=@n12, Alpha=Get Thigh Rot L() -> (Thigh Rot L:double), Translation="0,0,0", Rotation="-20,0,70", Scale="1,1,1") {
    rotationMode=BMM_Additive
    rotationSpace=BCS_BoneSpace
    scaleMode=BMM_Additive
    scaleSpace=BCS_BoneSpace
    translationMode=BMM_Additive
    translationSpace=BCS_BoneSpace
    alphaBoolBlend {blendOption=Linear}
    alphaScaleBiasClamp {bMapRange=true, inRange {max=0, min=-100}, outRange {max=0, min=1}}
    boneToModify {boneName=dyn_front_flap_master}
    rotation {pitch=-20, roll=70}
   }
   n14 = "Transform (Modify) Bone - Bone: dyn_front_flap_master"(ComponentPose=@n13, Alpha=Get Thigh Rot R(), Translation="0,0,0", Rotation="20,0,-70", Scale="1,1,1") {
    rotationMode=BMM_Additive
    rotationSpace=BCS_BoneSpace
    scaleMode=BMM_Additive
    scaleSpace=BCS_BoneSpace
    translationMode=BMM_Additive
    translationSpace=BCS_BoneSpace
    alphaBoolBlend {blendOption=Linear}
    alphaScaleBiasClamp {bMapRange=true, inRange {max=0, min=-100}, outRange {max=0, min=1}}
    boneToModify {boneName=dyn_front_flap_master}
    rotation {pitch=20, roll=-70}
   }
   n15 = "Constraint - dyn_belt_01_master (P)"(ComponentPose=@n14, Alpha=1, ConstraintWeights_0=1) {
    alphaBoolBlend {blendOption=Linear}
    boneToModify {boneName=dyn_belt_01_master}
    constraintSetup [{targetBone {boneName=thigh_l}}]
    constraintWeights [1]
   }
   n16 = "Transform (Modify) Bone - Bone: dyn_belt_01_master"(ComponentPose=@n15, Alpha=1, Translation="3,8,-20", Rotation="0,40,0", Scale="1,1,1") {
    rotationMode=BMM_Additive
    rotationSpace=BCS_BoneSpace
    translationMode=BMM_Additive
    translationSpace=BCS_BoneSpace
    alphaBoolBlend {blendOption=Linear}
    boneToModify {boneName=dyn_belt_01_master}
    rotation {yaw=40}
    translation=<3 8 -20>
   }
   n17 = RigidBody(ComponentPose=@n16) {
    overridePhysicsAsset=&/Space_Warrior__Belt_Crouch_Physics
    simulationSpace=BaseBoneSpace
    alphaBoolBlend {blendOption=Linear}
    baseBoneRef {boneName=root}
    componentLinearAccScale=<1 1 1>
    componentLinearVelScale=<0.3 0.3 0.3>
   }
   n18 = Control Rig(Source=Component To Local(ComponentPose=@n17)) {
    alphaBoolBlend {blendOption=Linear}
    controlRigAssetReference {blueprintRigClass=&Space_Warrior_CR_C}
   }
   Save cached pose 'crouch_base'(Pose=@n18) {cacheName=crouch_base}
  ]}
 EventGraph {nodes [
   Event BlueprintUpdateAnimation()
   > n1 = Sequence!(execute) -> (then_0, then_1, then_2:Exec, then_3:Exec, then_4:Exec, then_5:Exec)
   n2 = GetOwnerSkeletalMeshComponent()
   n3 = BreakRotator(InRot=BreakTransform(InTransform=GetDeltaTransformFromRefPose(self=@n2, BoneName=thigh_r, BaseName=pelvis)).Rotation)
   Set Thigh Rot R(execute=@n1.then_1, Thigh Rot R:double=@n3.Yaw) -> (Output_Get:double)
   > Set Thigh Roll_R(Thigh Roll_R:double=@n3) -> (Output_Get:double)
   n4 = GetDeltaTransformFromRefPose(self=GetOwnerSkeletalMeshComponent(), BoneName=thigh_l, BaseName=pelvis)
   Set Thigh Rot L(execute=@n1.then_0, Thigh Rot L:double=BreakRotator(InRot=BreakTransform(InTransform=@n4).Rotation).Yaw) -> (Output_Get:double)
   TryGetPawnOwner()
  ]}
]
variables [
 Thigh Roll_R {type=number}
 Thigh Rot L {type=number}
 Thigh Rot R {type=number}
]
