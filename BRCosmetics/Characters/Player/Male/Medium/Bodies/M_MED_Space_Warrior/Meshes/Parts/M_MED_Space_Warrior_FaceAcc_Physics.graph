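# Physics asset for the face accessory of the male medium Space Warrior character, defining collision and joints for two bandana tails.
# Each tail is a two-box chain, with the first segment constrained to the head capsule and the second to the first.
# Twist is locked on every joint, and swing is tight at the head and looser at the tips.
# A separate box body covers the dyn_collar bone and has no constraint.
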PhysicsAsset /BRCosmetics/Characters/Player/Male/Medium/Bodies/M_MED_Space_Warrior/Meshes/Parts/M_MED_Space_Warrior_FaceAcc_Physics
bodies {
 bandana_l_01 [{shapeType=Box, extentX=4, extentY=2, extentZ=1.5, center=<4.4674 -0.2405 0.6799>, rotation=<rot 8.5116 19.2581 2.96>}]
 bandana_l_02 [{shapeType=Box, extentX=2.5, extentY=1.5, extentZ=2.5, center=<2.3262 1.1263 -1.0297>, rotation=<rot -22.5241 9.2705 0>}]
 bandana_r_01 [{shapeType=Box, extentX=6, extentY=2, extentZ=2, center=<3.8193 -0.336 0.0001>}]
 bandana_r_02 [{shapeType=Box, extentX=2.5, extentY=2, extentZ=2, center=<1.4465 0 0>}]
 dyn_collar [{shapeType=Box, extentX=8.51, extentY=7.51, extentZ=12.51, center=<2.7089 -0.9086 -0.0001>, rotation=<rot 0 25.5083 0>}]
 head [{shapeType=Capsule, length=10.8765, radius=7.6842, center=<-0.4409 -0.4155 0>}]
}
constraints [
 {bone1Name=bandana_r_01, bone2Name=head, swing1LimitDegrees=1, swing2LimitDegrees=20, twistLimitDegrees=1, twistMotion=Locked}
 {bone1Name=bandana_l_01, bone2Name=head, swing1LimitDegrees=1, swing2LimitDegrees=20, twistLimitDegrees=1, twistMotion=Locked}
 {bone1Name=bandana_r_02, bone2Name=bandana_r_01, swing1LimitDegrees=10, swing2LimitDegrees=40, twistLimitDegrees=45, twistMotion=Locked}
 {bone1Name=bandana_l_02, bone2Name=bandana_l_01, swing1LimitDegrees=10, swing2LimitDegrees=40, twistLimitDegrees=45, twistMotion=Locked}
]
