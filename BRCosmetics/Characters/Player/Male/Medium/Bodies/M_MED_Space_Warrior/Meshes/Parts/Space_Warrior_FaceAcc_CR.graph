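# Control Rig for the Space Warrior face accessory part that procedurally shifts the dyn_collar bone as the head rotates.
# It reads the head's rotation relative to spine_05, negates the Euler Z angle and remaps it with clamping to a blend factor.
# That factor scales local X and Y translation offsets applied additively to the collar, so it moves for one direction of head rotation only.
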ControlRigBlueprint /BRCosmetics/Characters/Player/Male/Medium/Bodies/M_MED_Space_Warrior/Meshes/Parts/Space_Warrior_FaceAcc_CR
graphs [
 RigVMFunctionLibrary {links [], nodes []}
 RigVMModel {nodes [
   RigUnit_BeginExecution = &RigUnit_BeginExecution()
   n1 = &RigUnit_GetRelativeBoneTransform(Bone=head, Space=spine_05)
   n2 = &RigVMFunction_MathFloatRemap(
    SourceMaximum=40
    SourceMinimum=0
    TargetMinimum=0
    bClamp=true
    Value=&RigVMFunction_MathFloatMul(B=-1, A=&RigVMFunction_MathQuaternionToEuler(RotationOrder=ZXY, Value=@n1.Transform.Rotation).Result.Z) {template="Multiply::Execute(in A,in B,out Result)"}.Result
   )
   /Script/ControlRig.RigUnit_ModifyBoneTransforms(
    BoneToModify=((Bone="dyn_collar",Transform=(Rotation=(X=0,Y=0,Z=0,W=1),Translation=(X=0,Y=0,Z=0),Scale3D=(X=1,Y=1,Z=1))))
    Mode=AdditiveLocal
    Weight=1
    WeightMaximum=1
    WeightMinimum=0
    BoneToModify.0.Transform.Translation.X=&RigVMFunction_MathFloatMul(B=-6, A=@n2.Result) {template="Multiply::Execute(in A,in B,out Result)"}.Result
    BoneToModify.0.Transform.Translation.Y=&RigVMFunction_MathFloatMul(B=5, A=@n2.Result) {template="Multiply::Execute(in A,in B,out Result)"}.Result
    ExecutePin=@RigUnit_BeginExecution.ExecutePin
   ) {is=struct RigVMUnitNode}
  ]}
]
hierarchy [
 root {loc=<0 0 0>, rot=<0 0 0 1>}
 pelvis {parent=root, loc=<0 -0.7071 93.1191>, rot=<0.0127 0.707 -0.0127 -0.707>}
 spine_01 {parent=pelvis, loc=<6.7494 -0.6388 0>, rot=<0 0 0.034 -0.9994>}
 spine_02 {parent=spine_01, loc=<7.2518 -0.035 0>, rot=<0 0 0.0238 -0.9997>}
 spine_03 {parent=spine_02, loc=<6.0104 0 0>, rot=<0 0 0.0149 -0.9999>}
 spine_04 {parent=spine_03, loc=<8.1024 0 0>, rot=<0 0 -0.0132 -0.9999>}
 spine_05 {parent=spine_04, loc=<10.0146 0 0>, rot=<0 0 -0.0015 -1>}
 clavicle_l {parent=spine_05, loc=<4.3575 -0.1466 -2.3223>, rot=<0.0465 -0.7289 -0.009 -0.6829>}
 clavicle_r {parent=spine_05, loc=<4.3575 -0.1466 2.3223>, rot=<0.7289 0.0465 -0.6829 0.009>}
 neck_01 {parent=spine_05, loc=<11.3138 0.1529 0>, rot=<0 0 -0.2003 -0.9797>}
 neck_02 {parent=neck_01, loc=<4.5627 0 0>, rot=<0 0 0.0504 -0.9987>}
 head {parent=neck_02, loc=<5.3914 0 0>, rot=<0 0 0.115 -0.9934>}
 faceAttach {parent=head, loc=<0.6201 0.9296 0>, rot=<0.0082 -0.7071 -0.0082 -0.7071>, scale=<0.89 0.89 0.89>}
 C_brow_mid {parent=faceAttach, loc=<-0.0098 9.805 5.2629>, rot=<0 0.0163 -0.0032 -0.9999>}
 L_brow_mid {parent=faceAttach, loc=<3.1743 9.3003 5.8712>, rot=<-0.0001 -0.0043 0.2553 -0.9669>}
 L_brow_outer {parent=faceAttach, loc=<5.2898 7.463 5.9468>, rot=<0.0462 0.0542 0.4308 -0.8996>}
 L_cheek_inner {parent=faceAttach, loc=<4.2155 7.9832 -0.031>, rot=<-0.0773 -0.2864 0.3159 -0.9012>}
 C_lip_upper_mid {parent=faceAttach, loc=<-0.0006 9.9715 -2.5661>, rot=<0.2225 0.0028 -0.001 -0.9749>}
 L_lip_upper_outer {parent=faceAttach, loc=<1.5171 9.303 -2.5836>, rot=<0.0326 -0.0821 0.2662 -0.9599>}
 L_lip_corner {parent=faceAttach, loc=<2.7441 8.3749 -2.7337>, rot=<0.0025 0.0085 -0.0146 -0.9999>}
 C_nose_bridge {parent=faceAttach, loc=<-0.0006 11.3598 2.4064>, rot=<-0.2218 -0.003 -0.0008 -0.9751>}
 R_brow_mid {parent=faceAttach, loc=<-3.174 9.3002 5.872>, rot=<-0.0001 0.0042 -0.2553 -0.9669>}
 R_brow_outer {parent=faceAttach, loc=<-5.2897 7.4629 5.9468>, rot=<0.0088 -0.0072 -0.4408 -0.8976>}
 R_cheek_inner {parent=faceAttach, loc=<-4.2155 7.9832 -0.031>, rot=<-0.0773 0.2864 -0.3159 -0.9012>}
 R_lip_corner {parent=faceAttach, loc=<-2.5911 8.5689 -2.7384>, rot=<0.0028 -0.0081 0.0146 -0.9999>}
 C_jaw {parent=faceAttach, loc=<-0.0168 1.4731 -0.9724>, rot=<0 0 0 1>}
 L_lip_lower_outer {parent=C_jaw, loc=<1.5436 7.7226 -1.6161>, rot=<0.024 0.1045 0.3502 -0.9305>}
 C_lip_lower_mid {parent=C_jaw, loc=<0.0091 8.235 -1.9067>, rot=<0 -0.0107 -0.0023 -0.9999>}
 R_lip_lower_outer {parent=C_jaw, loc=<-1.4971 7.6718 -1.7667>, rot=<0.0162 -0.0621 -0.3189 -0.9456>}
 teeth_lower {parent=C_jaw, loc=<0.0168 7.3593 -2.7448>, rot=<0 0 0 1>}
 R_eye {parent=faceAttach, loc=<-3.1722 6.8616 4.4346>, rot=<0 0 0.9998 -0.0187>}
 R_eye_lid_upper_mid {parent=R_eye, loc=<-0.0396 -1.1714 0.2631>, rot=<0.0523 -0.0009 0.0167 -0.9985>}
 R_eye_lid_lower_mid {parent=R_eye, loc=<-0.0085 -0.9425 -0.7051>, rot=<0.0329 -0.0005 0.0167 -0.9993>}
 L_eye {parent=faceAttach, loc=<3.1714 6.965 4.4308>, rot=<0 0 0.9998 0.0187>}
 L_eye_lid_upper_mid {parent=L_eye, loc=<0.0251 -1.1623 0.2388>, rot=<0.0523 0 0 -0.9986>}
 L_eye_lid_lower_mid {parent=L_eye, loc=<0.0223 -0.923 -0.7488>, rot=<0.0329 0 0 -0.9995>}
 R_lip_upper_outer {parent=faceAttach, loc=<-1.5176 9.3024 -2.5836>, rot=<0.0354 0.0744 -0.2633 -0.9612>}
 bandana_r_01 {parent=head, loc=<5.7092 -7.678 0.0825>, rot=<0.0652 -0.9572 -0.0168 0.2815>}
 bandana_r_02 {parent=bandana_r_01, loc=<5.6491 -0.1684 1.1265>, rot=<-0.0048 0.1733 -0.0984 -0.9799>}
 bandana_l_01 {parent=head, loc=<5.3634 -7.6625 -0.7193>, rot=<0.1965 -0.9765 -0.0273 -0.0845>}
 bandana_l_02 {parent=bandana_l_01, loc=<4.7879 -0.3472 0.8078>, rot=<0.0165 -0.0419 -0.0971 -0.9943>}
 dyn_collar {parent=spine_05, loc=<18.2329 -12.0805 0>, rot=<0 0 -0.4257 -0.9049>}
]
