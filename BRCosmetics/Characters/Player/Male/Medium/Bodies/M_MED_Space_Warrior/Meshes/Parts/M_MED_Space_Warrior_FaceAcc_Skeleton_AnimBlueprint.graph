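# Animation Blueprint for the Space Warrior face accessory mesh, posing its bandana on top of the pose copied from the parent mesh.
# Both branches run a face accessory Control Rig, then either additive bandana bone offsets, with the right side weakened in the front end, or RigidBody simulation relative to spine_05.
# Each update reads head yaw relative to spine_05, blending to the offsets below a small threshold and to physics otherwise.
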AnimBlueprint /BRCosmetics/Characters/Player/Male/Medium/Bodies/M_MED_Space_Warrior/Meshes/Parts/M_MED_Space_Warrior_FaceAcc_Skeleton_AnimBlueprint
target_skeleton=&/M_MED_Space_Warrior_FaceAcc_Skeleton
graphs [
 AnimGraph {nodes [
   n1 = Control Rig(Source=Copy Pose From Mesh() {bUseAttachedParent=true}) {
    alphaBoolBlend {blendOption=Linear}
    controlRigAssetReference {blueprintRigClass=&/Space_Warrior_FaceAcc_CR.Space_Warrior_FaceAcc_CR_C}
   }
   n2 = SelectFloat(A=0.1, B=1, bPickA=Get bIsInFrontEnd())
   n3 = "Transform (Modify) Bone - Bone: bandana_r_01"(ComponentPose=Local To Component(LocalPose=@n1), Alpha=@n2, Translation="0,0,0", Rotation="20,0.000,-8", Scale="1,1,1") {
    rotationMode=BMM_Additive
    translationMode=BMM_Additive
    alphaBoolBlend {blendOption=Linear}
    boneToModify {boneName=bandana_r_01}
    rotation {pitch=20, roll=-8}
   }
   n4 = "Transform (Modify) Bone - Bone: bandana_r_02"(ComponentPose=@n3, Alpha=@n2, Translation="0,0,1", Rotation="0,0,0", Scale="1,1,1") {
    rotationMode=BMM_Additive
    translationMode=BMM_Additive
    alphaBoolBlend {blendOption=Linear}
    boneToModify {boneName=bandana_r_02}
    translation {z=1}
   }
   n5 = "Transform (Modify) Bone - Bone: bandana_l_01"(ComponentPose=@n4, Alpha=1, Translation="0,0,0", Rotation="0,0.000,-20", Scale="1,1,1") {
    rotationMode=BMM_Additive
    translationMode=BMM_Additive
    alphaBoolBlend {blendOption=Linear}
    boneToModify {boneName=bandana_l_01}
    rotation {roll=-20}
   }
   n6 = "Transform (Modify) Bone - Bone: bandana_l_02"(ComponentPose=@n5, Alpha=1, Translation="0,0,0", Rotation="0,0,0", Scale="1,1,1") {rotationMode=BMM_Additive, translationMode=BMM_Additive, alphaBoolBlend {blendOption=Linear}, boneToModify {boneName=bandana_l_02}}
   n7 = Control Rig(Source=Copy Pose From Mesh() {bUseAttachedParent=true}) {
    alphaBoolBlend {blendOption=Linear}
    controlRigAssetReference {blueprintRigClass=&Space_Warrior_FaceAcc_CR_C}
   }
   n8 = RigidBody(ComponentPose=Local To Component(LocalPose=@n7)) {
    overridePhysicsAsset=&/M_MED_Space_Warrior_FaceAcc_Physics
    simulationSpace=BaseBoneSpace
    alphaBoolBlend {blendOption=Linear}
    baseBoneRef {boneName=spine_05}
    componentLinearAccScale=<1 1 1>
    componentLinearVelScale=<0.3 0.3 0.3>
   }
   n9 = Get headRot() -> (headRot:double)
   Output Pose(Result=Blend Poses by bool(
     BlendPose_0=Component To Local(ComponentPose=@n6)
     BlendPose_1=Component To Local(ComponentPose=@n8)
     BlendTime_0=0.3
     BlendTime_1=0.3
     bActiveValue="float > float"(A=0.1, B=@n9) {is=title K2Node_CallFunction}
    ) {blendType=Linear}) {layerGroup=None}
  ]}
 EventGraph {nodes [
   Event BlueprintUpdateAnimation()
   n1 = GetDeltaTransformFromRefPose(self=GetOwnerSkeletalMeshComponent(), BoneName=head, BaseName=spine_05)
   > Set headRot(headRot:double=BreakRotator(InRot=BreakTransform(InTransform=@n1).Rotation).Yaw) -> (Output_Get:double)
   TryGetPawnOwner()
  ]}
]
variables [headRot {type=number}]
